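# Unlit, masked, two-sided sprite material for Landmass landscape Blueprint brushes, apparently a brush marker icon.
# The Sprite function builds a world position offset sized by the object's X scale, plus a Z lift of half that scale.
# A texture sample references the editor's terrain icon.
# No links to material outputs appear, and two duplicate Sprite Size parameters feed nothing.
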Material /Landmass/Landscape/BlueprintBrushes/Materials/Sprite/M_BrushSprite_01
graph {
 properties {blendMode=BLEND_Masked, pixelDepthOffsetMode=PDOM_Legacy, shadingModel=MSM_Unlit, twoSided=true, usedShadingModels=Unlit}
 nodes [
  n1 = MaterialFunctionCall() {materialFunction=&ObjectScale}
  Add(
   A=MaterialFunctionCall!(Center Location, 0-1 UVs, XY scale=@n1.Scale X, OverridePivot, "Normals (Optional)") {materialFunction=&Sprite}.World position offset
   B=MaterialFunctionCall(Z=Multiply(A=@n1.Scale X) {constB=0.5}) {materialFunction=&MakeFloat3}.Result
  )
  ScalarParameter() {defaultValue=4096, parameterName=Sprite Size}
  ScalarParameter() {defaultValue=4096, parameterName=Sprite Size}
  TextureSample() {texture=/Engine/EditorResources/S_Terrain}
 ]
}
